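; Controller component for the Jam music system that manages a player's UI, controls and state within a jam session.
; On session connect or BeginPlay, it applies default privacy and requests UI scenes filtered by playlist tags and CVars.
; Async Jamming-tag listeners toggle emote controls and a GameplayEffect; disconnect and EndPlay tear these down, with effect removal CVar-gated.
; It also relays stop-loop and privacy changes to the server.
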(event EventBeginPlay
  (Utilities|FlowControl|Sequence
    (:then_0
      (|CallParentFunction))
    (:then_1
      (CallFunction|FetchSceneCoordinator))
    (:then_2
      (Class|JamSessionObserverInterface|GetCurrentJamSession
        (:True
          (Utilities|FlowControl|Sequence
            (:then_0
              ; Apply default jam privacy setting if this is a new jam (check for no jammer, an invalid first jammer, or if the local player is the first jammer)
              (Networking|CanExecuteCosmeticEvents
                (:True
                  (if (Utilities|Array|IsNotEmpty (Mojo|JamSession|GetAllJammers (Class|JamSessionInterface|GetMusicManager (stale EventNotifyConnectingToJamSession JamSession))))
                    (bind _returnvalue (UniqueNetID|IsValidNetID (Utilities|Array|Get(aref) (Mojo|JamSession|GetAllJammers (Class|JamSessionInterface|GetMusicManager (stale EventNotifyConnectingToJamSession JamSession))))))
                    (if (or (not _returnvalue) (== (Utilities|Array|Get(aref) (Mojo|JamSession|GetAllJammers (Class|JamSessionInterface|GetMusicManager (stale EventNotifyConnectingToJamSession JamSession)))) (Online|UniqueNetIdRepl|GetNetIDFromContextObject (Components|GetOwner))))
                      _))
                  (Jam|ApplySessionDefaultPrivacytoCurrentJam))))
            (:then_1
              (CallFunction|AddUIScenes)
              (bind _returnvalue_2 (CallFunction|GetControllerOfClass "/Script/Engine.PlayerController"))
              (Utilities|IsValid _returnvalue_2
                (:Is%20Valid
                  (Utilities|FlowControl|Sequence
                    (:then_0
                      (Utilities|IsValid (Variables|Default|GetJammingTagAddedHandle)
                        (:Is%20Not%20Valid
                          (bind _asynctaskproxy (Ability|Tasks|WaitGameplayTagAddToActor _returnvalue_2 "(TagName=\"Gameplay.Action.Player.Jam.Jamming\")")
                            (:then
                              (Variables|Default|SetJammingTagAddedHandle _asynctaskproxy))
                            (:Added
                              (JamControls|SetWantsJamControls true) ; FORT-778016 - @See above
                              (CallFunction|ApplyJammingGE))))))
                    (:then_1
                      (Utilities|IsValid (Variables|Default|GetJammingTagRemovedHandle)
                        (:Is%20Not%20Valid
                          (bind _asynctaskproxy_2 (Ability|Tasks|WaitGameplayTagRemoveFromActor _returnvalue_2 "(TagName=\"Gameplay.Action.Player.Jam.Jamming\")")
                            (:then
                              (Variables|Default|SetJammingTagRemovedHandle _asynctaskproxy_2))
                            (:Removed
                              (JamControls|SetWantsJamControls false) ; FORT-778016 - @See above
                              (CallFunction|RemoveJammingGE))))))))))))))))

(event EventNotifyConnectingToJamSession
  (params (JamSession :type Jam%20Session%20Interface%20Interface%20%28by%20ref%29))
  (bind _returnvalue (Class|JamSessionInterface|GetMusicManager JamSession))
  (bind _returnvalue_2 (Mojo|JamSession|GetAllJammers _returnvalue))
  (bind _output (Utilities|Array|Get(aref) _returnvalue_2))
  (Utilities|FlowControl|Sequence
    (:then_0
      ; Apply default jam privacy setting if this is a new jam (check for no jammer, an invalid first jammer, or if the local player is the first jammer)
      (Networking|CanExecuteCosmeticEvents
        (:True
          (if (Utilities|Array|IsNotEmpty _returnvalue_2)
            (bind _returnvalue_3 (UniqueNetID|IsValidNetID _output))
            (if (or (not _returnvalue_3) (== _output (Online|UniqueNetIdRepl|GetNetIDFromContextObject (Components|GetOwner))))
              _))
          (Jam|ApplySessionDefaultPrivacytoCurrentJam))))
    (:then_1
      (CallFunction|AddUIScenes)
      (bind _returnvalue_4 (CallFunction|GetControllerOfClass "/Script/Engine.PlayerController"))
      (Utilities|IsValid _returnvalue_4
        (:Is%20Valid
          (Utilities|FlowControl|Sequence
            (:then_0
              (Utilities|IsValid (Variables|Default|GetJammingTagAddedHandle)
                (:Is%20Not%20Valid
                  (bind _asynctaskproxy (Ability|Tasks|WaitGameplayTagAddToActor _returnvalue_4 "(TagName=\"Gameplay.Action.Player.Jam.Jamming\")")
                    (:then
                      (Variables|Default|SetJammingTagAddedHandle _asynctaskproxy))
                    (:Added
                      (JamControls|SetWantsJamControls true) ; FORT-778016 - @See above
                      (CallFunction|ApplyJammingGE))))))
            (:then_1
              (Utilities|IsValid (Variables|Default|GetJammingTagRemovedHandle)
                (:Is%20Not%20Valid
                  (bind _asynctaskproxy_2 (Ability|Tasks|WaitGameplayTagRemoveFromActor _returnvalue_4 "(TagName=\"Gameplay.Action.Player.Jam.Jamming\")")
                    (:then
                      (Variables|Default|SetJammingTagRemovedHandle _asynctaskproxy_2))
                    (:Removed
                      (JamControls|SetWantsJamControls false) ; FORT-778016 - @See above
                      (CallFunction|RemoveJammingGE))))))))))))

(event EventNotifyDisconnectingFromJamSession
  (params (JamSession :type Jam%20Session%20Interface%20Interface%20%28by%20ref%29))
  (bind _jammingtagremovedhandle (Variables|Default|GetJammingTagRemovedHandle))
  (bind _jammingtagaddedhandle (Variables|Default|GetJammingTagAddedHandle))
  (CallFunction|RemoveUIScenes)
  (JamControls|SetWantsJamControls false)
  (Utilities|FlowControl|Sequence
    (:then_0
      (Utilities|IsValid _jammingtagaddedhandle
        (:Is%20Valid
          (Ability|Async|EndAction _jammingtagaddedhandle)
          (Variables|Default|SetJammingTagAddedHandle 0))))
    (:then_1
      (Utilities|IsValid _jammingtagremovedhandle
        (:Is%20Valid
          (Ability|Async|EndAction _jammingtagremovedhandle)
          (Variables|Default|SetJammingTagRemovedHandle 0))))
    (:then_2
      ; FORT-1026884 - remove the jamming GE when leaving the JamSession or ending Play
      (bind _returnvalue (Development|GetConsoleVariableBoolValue "Fort.Rollback.BPJamControllerComponentSmarterGECleanup"))
      (if _returnvalue
        (CallFunction|RemoveJammingGE)))))

(event EventEndPlay
  (params (EndPlayReason :type EEndPlayReason%20Enum :default "Destroyed"))
  (bind _jammingtagremovedhandle (Variables|Default|GetJammingTagRemovedHandle))
  (bind _jammingtagaddedhandle (Variables|Default|GetJammingTagAddedHandle))
  (Utilities|FlowControl|Sequence
    (:then_0
      (Class|JamSessionObserverInterface|GetCurrentJamSession
        (:True
          (CallFunction|RemoveUIScenes))
        (:False
          _))
      (JamControls|SetWantsJamControls false)
      (Utilities|FlowControl|Sequence
        (:then_0
          (Utilities|IsValid _jammingtagaddedhandle
            (:Is%20Valid
              (Ability|Async|EndAction _jammingtagaddedhandle)
              (Variables|Default|SetJammingTagAddedHandle 0))))
        (:then_1
          (Utilities|IsValid _jammingtagremovedhandle
            (:Is%20Valid
              (Ability|Async|EndAction _jammingtagremovedhandle)
              (Variables|Default|SetJammingTagRemovedHandle 0))))
        (:then_2
          ; FORT-1026884 - remove the jamming GE when leaving the JamSession or ending Play
          (bind _returnvalue (Development|GetConsoleVariableBoolValue "Fort.Rollback.BPJamControllerComponentSmarterGECleanup"))
          (if _returnvalue
            (CallFunction|RemoveJammingGE)))))
    (:then_1
      (|CallParentFunction EndPlayReason))))

(event Custom|StopLoop
  ;[1 0.3364 0.806] RPCs
  (CallFunction|StopLoopServer))

(event Custom|StopLoop_Server
  ;[1 0.3364 0.806] RPCs
  (Utilities|SwitchHasAuthority
    (:Authority
      (bind _returnvalue (Class|JamSessionObserverInterface|GetCurrentJamSession)
        (:True
          (Utilities|IsValid (Class|JamSessionInterface|GetMusicManager _returnvalue)
            (:Is%20Valid
              (bind _returnvalue_2 (CallFunction|GetControllerOfClass "/Script/FortniteGame.FortPlayerController"))
              (Utilities|IsValid _returnvalue_2
                (:Is%20Valid
                  (Mojo|JamSession|StopLoopforPlayer (Class|JamSessionInterface|GetMusicManager _returnvalue) (Default|GetGameAccountId _returnvalue_2) true true))))))))))

(event EnhancedInputActionIA_Jam_StopLoop
  (params (ActionValue :type Boolean) (ElapsedSeconds :type Float%20%28double-precision%29) (TriggeredSeconds :type Float%20%28double-precision%29) (InputAction :type Abstractable%20Input%20Action%20Object%20Reference :default "/FMJamSystem/Input/InputActions/IA_Jam_StopLoop.IA_Jam_StopLoop")))

(event EnhancedInputActionIA_Jam_StopLoop_Touch
  (params (ActionValue :type Boolean) (ElapsedSeconds :type Float%20%28double-precision%29) (TriggeredSeconds :type Float%20%28double-precision%29) (InputAction :type Abstractable%20Input%20Action%20Object%20Reference :default "/FMJamSystem/Input/InputActions/IA_Jam_StopLoop_Touch.IA_Jam_StopLoop_Touch")))

(event Custom|SetCurrentJamPrivacyForMusicManager
  (params (JamPrivacy :type EJamPrivacySetting%20Enum :default "Public"))
  ;[1 0.3364 0.806] RPCs
  (CallFunction|ServerSetJamPrivacy JamPrivacy)
  (Jam|SetSessionDefaultPrivacySetting JamPrivacy))

(fn AddUIScenes
  (Networking|CanExecuteCosmeticEvents
    (:True
      (Development|PrintText (Utilities|Text|FormatText "INVTEXT(\"{obj}: Adding All UIScenes\")" self) false)
      (bind _scenestoadd (CallFunction|GetUIScenes))
      ; This prevents us from removing a scene that another component just added in niche cases - FORT-745727
      (bind _dynamicuicoordinator (Variables|SceneCoordination|GetDynamicUICoordinator)
        (:then
          (bind _returnvalue (SparksUIScenes|CoordinatedRequests|Scenes|AddScenes _dynamicuicoordinator self _scenestoadd))
          (Utilities|Array|Add (Variables|SceneCoordination|GetDynamicUI%20Scene%20Requests) _returnvalue))
        (:else)))))

(fn AllowedToEnterMusicPlayspace
  (params (Playspace :type Sparks%20Music%20Playspace%20Object%20Reference))
  (return true))

(fn ApplyJammingGE
  (bind _returnvalue (CallFunction|GetControllerOfClass "/Script/Engine.Controller"))
  (Utilities|IsValid _returnvalue
    (:Is%20Valid
      (Fortnite|ApplyGameplayEffectsToActor _returnvalue (Utilities|Array|MakeArray (Variables|Default|GetJammingGE)) (Transformation|GetActorLocation _returnvalue) false 0 (Ability|GetAbilitySystemComponent _returnvalue))
      ;[1 0.5 0.5] GE has been applied
      (return))
    (:Is%20Not%20Valid
      ;[1 0 0] Could not get Controller
      (return))))

(fn FetchSceneCoordinator
  (bind _returnvalue (CallFunction|GetControllerOfClass "/Script/Engine.PlayerController"))
  (Utilities|IsValid _returnvalue
    (:Is%20Valid
      (Variables|SceneCoordination|SetDynamicUICoordinator (SparksUIScenes|CoordinatedRequests|GetDynamicUISceneCoordinator _returnvalue)))))

(fn GetUIScenes
  ;[0.04337 1 0.438] Don't add ANY ui if this JamSession won't provide jamming
  (bind _returnvalue (Class|JamSessionObserverInterface|GetCurrentJamSession)
    (:True
      (if (not (Class|JamSessionInterface|BlocksJamming _returnvalue))
        ; Save The World (STW/Campaign) doesn't use playlists
        (if (not (FortGlobals|IsinSubGame "Campaign")) :!label _label
          (LocalVariables|SetPlaylist%20Tags (Class|FortGameStateAthena|GetAthenaPlaylistContextTags (Fortnite|Athena|GameState|GetGameStateAthena))))
        (for _array_element (Variables|Default|GetUIScenes)
          (if (or (GameplayTags|DoesContainerMatchTagQuery (LocalVariables|GetPlaylist%20Tags) (Utilities|Struct|BreakJamUIScenebyTags _array_element)) (GameplayTags|IsTagQueryEmpty (Utilities|Struct|BreakJamUIScenebyTags _array_element)))
            (bind _has_all (CallFunction|HasAllCVars (.RequiredCVars_19_2E19D6E24076774D3965438DAA8EC693 (Utilities|Struct|BreakJamUIScenebyTags _array_element))))
            (if _has_all
              (bind _has_any (CallFunction|HasAnyCVars (.DisallowCVars_20_17C652094212A63701EAC9855C835645 (Utilities|Struct|BreakJamUIScenebyTags _array_element))))
              (if (not _has_any)
                (for _array_element_2 (.UIScenesToAdd_6_6A717A3F46A28CA9EA3C9E8C1909000C (Utilities|Struct|BreakJamUIScenebyTags _array_element))
                  (Utilities|Array|AddUnique (LocalVariables|GetApplicableScenes) _array_element_2)))))))
      (return (LocalVariables|GetApplicableScenes)))
    (:False
      _))
  (goto _label))

(fn HasAllCVars
  (params (CVars :type Array%20of%20Strings))
  (LocalVariables|SetRet true)
  (for _array_element CVars
    (bind _returnvalue (Development|GetConsoleVariableBoolValue _array_element))
    (if (not _returnvalue)
      (LocalVariables|SetRet false)))
  (return (LocalVariables|GetRet)))

(fn HasAnyCVars
  (params (CVars :type Array%20of%20Strings))
  (LocalVariables|SetRet false)
  (for _array_element CVars
    (bind _returnvalue (Development|GetConsoleVariableBoolValue _array_element))
    (if _returnvalue
      (LocalVariables|SetRet true)))
  (return (LocalVariables|GetRet)))

(fn ProcessChangeToWantsJamControls
  ; FORT-778016 - only enable the controls when we both WANT and CAN have them.
  (if (Variables|JamControls|GetbWantsJamControls)
    (JamControls|InternalDONOTUSE|AddEmoteControls) ; FORT-778016 this site is fine
    (else
      (JamControls|InternalDONOTUSE|RemoveEmoteControls) ; FORT-778016 this site is fine
    )))

(fn RemoveEmoteControls
  (bind (_outfoundpriority _returnvalue) (Input|MappingQueries|HasMappingContext (PlayerController|LocalPlayerSubsystems|GetEnhancedInputLocalPlayerSubsystem (LocalVariables|GetThe%20Controller)) (Variables|EnhancedInput|GetJamControlsMap)))
  (bind _the_controller (LocalVariables|GetThe%20Controller))
  (bind _returnvalue_2 (PlayerController|LocalPlayerSubsystems|GetEnhancedInputLocalPlayerSubsystem _the_controller))
  (Networking|CanExecuteCosmeticEvents
    (:True
      (bind _output_get (LocalVariables|SetThe%20Controller (CallFunction|GetControllerOfClass "/Script/Engine.PlayerController")))
      (Utilities|IsValid _output_get
        (:Is%20Valid
          (Utilities|IsValid _returnvalue_2
            (:Is%20Valid
              (Utilities|FlowControl|Sequence
                (:then_0
                  ;[1 1 1] Remove the standard controls
                  (if _returnvalue
                    ; Remove the Input Mapping Context
                    (bind _returnvalue_3 (ManagedInputPrompts|GetInputPromptManager (LocalVariables|GetThe%20Controller)))
                    (ManagedInputPrompts|EnhancedInput|RemoveInputMappingContext _returnvalue_3 (Variables|EnhancedInput|GetJamControlsMap))
                    (ManagedInputPrompts|EnhancedInput|RemoveInputMappingContext _returnvalue_3 (Variables|EnhancedInput|GetJamControlsMap_Touch))
                    ; Unbind from the on input method changed event
                    (CommonInputSubsystem|UnbindEventfromOnInputMethodChanged (PlayerController|LocalPlayerSubsystems|GetCommonInputSubsystem (LocalVariables|GetThe%20Controller)) (EventDispatchers|CreateEvent :!event ToggleJamControlsMapTouch))))
                (:then_1
                  ; Remove our Input Prompts
                  (bind _returnvalue_4 (ManagedInputPrompts|GetInputPromptManager (LocalVariables|GetThe%20Controller)))
                  (ManagedInputPrompts|RemoveInputAliases _returnvalue_4 (Variables|Default|GetActive%20Input%20Aliases))
                  (Utilities|Array|Clear (Variables|Default|GetActive%20Input%20Aliases))
                  ;[1 0.5 0.5] Input mappings removed
                  (Development|PrintString "Input mappings removed" false true "(R=0.000000,G=0.660000,B=1.000000,A=1.000000)" 1e+01)
                  (return))))
            (:Is%20Not%20Valid
              ;[1 0 0] Could not get EnhancedInput
              (Development|PrintString "Could not get EnhancedInput" false true "(R=0.000000,G=0.660000,B=1.000000,A=1.000000)" 1e+01)
              (return))))
        (:Is%20Not%20Valid
          ;[1 0 0] Could not find controller
          (Development|PrintString "Failed to find Owning Controller" false true "(R=0.000000,G=0.660000,B=1.000000,A=1.000000)" 1e+01)
          (return))))
    (:False
      ;[1 0.5 0.5] Nothing to do on server
      (return))))

(fn RemoveJammingGE
  (bind _returnvalue (CallFunction|GetControllerOfClass "/Script/Engine.Controller"))
  (Utilities|IsValid _returnvalue
    (:Is%20Valid
      (GameplayEffects|RemoveActiveGameplayEffectbySourceEffect (Ability|GetAbilitySystemComponent _returnvalue) (Variables|Default|GetJammingGE) 0 -1)
      ;[1 0.5 0.5] All instances of the GE are removed
      (return))
    (:Is%20Not%20Valid
      ;[1 0 0] Could not get Controller
      (return))))

(fn RemoveUIScenes
  (Networking|CanExecuteCosmeticEvents
    (:True
      (Development|PrintText (Utilities|Text|FormatText "INVTEXT(\"{obj}: Removing All UIScenes\")" self) false)
      ; This prevents us from removing a scene that another component just added in niche cases - FORT-745727
      ;; These handle instances are what keeps our scene in the Dynamic UI Manager
      (Utilities|Array|Clear (Variables|SceneCoordination|GetDynamicUI%20Scene%20Requests)))))

(fn SetWantsJamControls
  (params (bNewWantsJamControls :type Boolean))
  ; FORT-778016 - Don't allow us to have the controls bound when we can't use them (@See 'Process Change to Wants Jam Controls')
  (if (!= bNewWantsJamControls (Variables|JamControls|GetbWantsJamControls))
    (Fortnite|FortShippingLog self (Utilities|String|ToString(Text) (Utilities|Text|FormatText "INVTEXT(\"FORT-778016: Changing \\\"Wants\\\" from {old} to {new}.\")" (Variables|JamControls|GetbWantsJamControls) (Variables|GetbNewWantsJamControls))))
    (Variables|JamControls|SetbWantsJamControls (Variables|GetbNewWantsJamControls))
    (JamControls|ProcessChangetoWantsJamControls)))
