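; Jam session reactive FX component that adds spawn logic for Jam companion auto jammers when a loop starts.
; GetPlayerVFXSpawnProps calls the parent first and adds custom handling only if that returns false.
; It finds the loop's music slot through the playspace's MusicManager and checks the companion slot tag.
; For companions, the companion playspace volume supplies a jam ring transform per loop type, where the default companion Niagara FX spawns unattached.
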(fn GetJamPlayspace
  (bind _asjam_playspace (Utilities|Casting|CastToJamPlayspace (Components|GetOwner))
    (:then
      (return _asjam_playspace))
    (:CastFailed
      (return 0))))

(fn GetMusicSlotByLoopInstanceId
  (params (LoopInstanceId :type Integer))
  (bind _jamplayspace (CallFunction|GetJamPlayspace))
  (bind _returnvalue (Actor|GetComponentbyClass _jamplayspace "/Script/FMJamSessionRuntime.JamSessionComponent_MusicManager"))
  (Utilities|IsValid _jamplayspace
    (:Is%20Valid
      (Utilities|IsValid _returnvalue
        (:Is%20Valid
          (return (Mojo|JamSession|GetMusicSlotbyLoopInstanceId _returnvalue LoopInstanceId)))
        (:Is%20Not%20Valid
          _)))
    (:Is%20Not%20Valid
      _))
  (return 0))

(fn GetPlayerVFXSpawnProps
  (params (Event :type Jam%20Event%20Jam%20Loop%20Started%20Structure%20%28by%20ref%29))
  (bind (_songshortname _catalogentry _looptype) (Utilities|Struct|BreakJamEventJamLoopStarted (Variables|GetEvent)))
  (bind _event (Variables|GetEvent))
  (bind (_songshortname_2 _catalogentry_2 _looptype_2 _playernetid _loopinstanceid _bisautojammer) (Utilities|Struct|BreakJamEventJamLoopStarted _event))
  (bind (_outcomponenttofollow _outspawnlocation _outspawnrotation _outspawnscale _outniagarasystemtospawn _returnvalue) (|CallParentFunction Event))
  (if _returnvalue
    _
    (elif _bisautojammer
      ;[0.5284 1 0.6603] Only have custom logic for Jam companion auto jammers
      (bind _output_get (LocalVariables|SetMusicSlot (CallFunction|GetMusicSlotbyLoopInstanceId _loopinstanceid)))
      (Utilities|IsValid _output_get
        (:Is%20Valid
          (if (Class|JamMusicSlot|HasGameplayTag (LocalVariables|GetMusicSlot) (JamSystem|GetJamCompanionSlotTag))
            ;[0.5284 1 0.6603] Need the Playspace Volume
            (bind _asbp_jam_playspace_volume_jam_companion (Utilities|Casting|CastToBP_JamPlayspace_Volume_JamCompanion (Mojo|JamPlayspace|GetJamPlayspaceVolume (CallFunction|GetJamPlayspace)))
              (:then
                ;[0.2614 0.4948 0.3267] Figure out where to put the jam ring.
                (bind _transform (Class|BPJamPlayspaceVolumeJamCompanion|GetJamRingTransformforTrackType _asbp_jam_playspace_volume_jam_companion _looptype))
                (return 0 (.location _transform) (.rotation _transform) (.scale _transform) (Variables|Default|GetDefaultJamCompanionFXToSpawn) true))
              (:CastFailed
                _))
            (else
              _))
          (return _outcomponenttofollow _outspawnlocation _outspawnrotation _outspawnscale _outniagarasystemtospawn _returnvalue))
        (:Is%20Not%20Valid
          _))
      (else
        _))))
