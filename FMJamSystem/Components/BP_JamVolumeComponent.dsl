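; Actor component that mutes jam audio in the FM Jam system when its owning player's music playspace signals it.
; On BeginPlay it casts the owner to FortPlayerState, resolves its SparksMusicPlayspace and stores that playspace's event router.
; It listens there for FMCoreEvent_MuteJamAudioChanged and passes the payload to SetMuteJam.
; SetMuteJam pushes or pops the BlockJam sound mix modifier, so muting works through the global mix.
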(event EventBeginPlay
  ;[0.9901 0.1279 1] Grabbing Event Router
  (bind _asfort_player_state (Utilities|Casting|CastToFortPlayerState (Components|GetOwner)))
  (bind _returnvalue (Class|PlayspaceLibrary|GetPlayspaceforActor _asfort_player_state _asfort_player_state))
  (bind _assparks_music_playspace (Utilities|Casting|CastToSparksMusicPlayspace _returnvalue))
  (bind _output_get (Variables|Default|SetEvent%20Router (Class|SparksMusicPlayspace|GetOwnedEventRouter _assparks_music_playspace)))
  ;[1 0.1424 0.856] Setting up Event Listener For Muting Jam Audio
  (bind (_asynctaskproxy _payload) (GameplayEventRouter|StartListeningToEvent _output_get "/Script/FMCoreRuntime.FMCoreEvent_MuteJamAudioChanged")
    (:OnEventReceived
      (Default|SetMuteJam (Utilities|Struct|BreakFMCoreEventMuteJamAudioChanged _payload)))))

(event EventTick
  (params (DeltaSeconds :type Float%20%28single-precision%29)))

(event EventEndPlay
  (params (EndPlayReason :type EEndPlayReason%20Enum :default "Destroyed")))

(fn SetMuteJam
  (params (Mute :type Boolean))
  (if Mute
    (Audio|PushSoundMixModifier "/FMJamSystem/Audio/FMJam_SoundMix_BlockJam.FMJam_SoundMix_BlockJam")
    (else
      (Audio|PopSoundMixModifier "/FMJamSystem/Audio/FMJam_SoundMix_BlockJam.FMJam_SoundMix_BlockJam"))))
